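# Sound class for Festivus event audio in Figment Season 9, named as a PIE suppression variant.
# It is a child of the Fort_Events_Festivus sound class and sets no properties of its own.
# The name suggests it groups sounds to be suppressed during Play In Editor, but no suppression logic appears here.
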SoundClass /Figment_S09/Athena/Sounds/Events/SoundClasses/Fort_Events_Festivus_PIESuppress
parent=&/Fort_Events_Festivus
